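# Unlit translucent energy effect material for the RelayStick keytar, used on Niagara sprite and mesh particles.
# Caustics texture pans over radial coordinates, with derivatives from plain UVs setting mips to avoid a polar seam.
# Particle color tints it, then time-of-day brightness and an emissive adjustment function shape the glow.
# Opacity is a ring made by masking one exponential radial gradient with another inverted one, boosted, saturated and scaled by particle alpha.
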Material /SparksCosmetics/Instrument/Keytar/RelayStick/Effects/M_Keytar_RelayStick_Energy
graph {
 properties {
  bUsedWithNiagaraMeshParticles=true
  bUsedWithNiagaraSprites=true
  blendMode=BLEND_Translucent
  pixelDepthOffsetMode=PDOM_Legacy
  shadingModel=MSM_Unlit
  usedShadingModels=Unlit
 }
 nodes [
  n1 = Constant() {r=10}
  n2 = Constant() {r=15}
  n3 = Constant() {r=10}
  n4 = Constant() {r=10}
  n5 = MaterialFunctionCall() {materialFunction=&VectorToRadialValue}
  n6 = Constant2Vector() {g=0.3, r=1}
  n7 = TextureCoordinate()
  n8 = TextureSample!(
   UVs=Panner(Coordinate=Multiply(A=@n5.Radial Coordinates, B=@n6)) {speedX=0.1, speedY=-0.2}
   Tex
   "DDX(UVs)"=DDX(Value=@n7)
   "DDY(UVs)"=DDY(Value=@n7)
   Apply View MipBias
  ) {mipValueMode=TMVM_Derivative, texture=&T_ShieldCaustics}
  n9 = ParticleColor()
  MaterialFunctionCall(Emissive=Multiply(
    A=MaterialFunctionCall(Color_Morning=@n1, Color_Day=@n2, Color_Evening=@n3, Color_Night=@n4) {materialFunction=&Color_Time_of_Day}.Result
    B=Multiply(A=@n8.R, B=@n9.RGB)
   )) {materialFunction=&MF_FX_EmissiveAdjustments}
  n10 = ScalarParameter() {defaultValue=0.4, parameterName=Param_3}
  n11 = ScalarParameter() {defaultValue=5, parameterName=Param_4}
  n12 = Multiply(
   A=MaterialFunctionCall(Radius=ScalarParameter() {defaultValue=0.5}, Density=ScalarParameter() {defaultValue=1, parameterName=Param_1}) {materialFunction=&RadialGradientExponential}.RadialGradientExponential
   B=OneMinus(None=MaterialFunctionCall(Radius=@n10, Density=@n11) {materialFunction=&RadialGradientExponential}.RadialGradientExponential)
  )
  Multiply(A=@n9.A, B=Saturate(None=Multiply(A=@n12) {constB=4}))
 ]
}
