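# Orb material for the RelayStick Keytar cosmetic, usable on Niagara mesh particles.
# An exponential radial gradient masks the Emissive output of MF_RelayStick_Orb, a function reused from the RelayStick pickaxe.
# The inverse gradient masks two summed panning samples of a PitGlass backpack noise texture, tinted cyan.
# Adding both layers places the shared orb look inside the gradient and drifting noise around it.
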Material /SparksCosmetics/Instrument/Keytar/RelayStick/Effects/M_Keytar_RelayStick_Orb
graph {
 properties {bUsedWithNiagaraMeshParticles=true, pixelDepthOffsetMode=PDOM_Legacy, usedShadingModels=Default Lit}
 nodes [
  n1 = TextureCoordinate() {uTiling=2}
  n2 = Constant2Vector() {g=0.5, r=1}
  n3 = ScalarParameter() {defaultValue=0.3}
  n4 = ScalarParameter() {defaultValue=3, parameterName=Param_1}
  n5 = TextureSample(UVs=Panner(Coordinate=TextureCoordinate() {uTiling=2.5, vTiling=1.5}) {speedX=0.05}) {
   samplerType=SAMPLERTYPE_LinearGrayscale
   texture=/Game/Accessories/FORT_Backpacks/Backpack_M_MED_PitGlass/FX/Textures/T_Backpack_PitGlass_Noise
  }
  n6 = TextureSample(UVs=Panner(Coordinate=TextureCoordinate() {uTiling=2}) {speedY=0.1}) {
   samplerType=SAMPLERTYPE_LinearGrayscale
   texture=&T_Backpack_PitGlass_Noise
  }
  n7 = Add(A=@n5.R, B=@n6.R)
  n8 = OneMinus(None=MaterialFunctionCall(
    UVs=TextureCoordinate() {uTiling=2}
    CenterPosition=Constant2Vector() {g=0.5, r=1}
    Radius=ScalarParameter() {defaultValue=0.3}
    Density=ScalarParameter() {defaultValue=3, parameterName=Param_1}
   ) {materialFunction=&RadialGradientExponential}.RadialGradientExponential)
  Add(
   A=Multiply(
    A=MaterialFunctionCall(UVs=@n1, CenterPosition=@n2, Radius=@n3, Density=@n4) {materialFunction=&RadialGradientExponential}.RadialGradientExponential
    B=MaterialFunctionCall!(AudioSpectrum, Fresnel, AudioAmplitude) {materialFunction=/BRCosmetics/Weapons/FORT_Melee/Pickaxe_RelayStick/Materials/MF_RelayStick_Orb}.Emissive
   )
   B=Multiply(A=Constant3Vector() {constant=<rgba 0 0.9 1 1>}, B=Multiply(A=@n7, B=@n8))
  )
 ]
}
