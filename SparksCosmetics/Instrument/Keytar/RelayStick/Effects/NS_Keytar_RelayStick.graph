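# Niagara effect for the RelayStick keytar cosmetic, drawing a persistent glowing, audio-reactive orb at a fixed offset.
# Four burst-spawned emitters layer an Orb mesh, a GlowEdge mesh, an Energy sprite and an OrbGlow sprite with a light.
# Glow layers share a looping rainbow-like color curve, and audio amplitude sets light radius.
# GlowEdge's material pulses on Sparks beat timing, switching to particle age when User.bIsFrontEndPreview is set.
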NiagaraSystem /SparksCosmetics/Instrument/Keytar/RelayStick/Effects/NS_Keytar_RelayStick
dependencies {
 usedDataInterfaces [&NiagaraDataInterfaceArrayFloat, &NiagaraDataInterfaceColorCurve, &NiagaraDataInterfaceCurve]
 usedDynamicInputs [&MakeVector, &SelectFloatFromArray]
 usedModules [&SystemState]
}
emitters {
 Energy {
  emitterUpdateScript [
   &EmitterState(Life Cycle Mode=::NewEnumerator1, Loop Behavior=::NewEnumerator1, Loop Duration Mode=::NewEnumerator1, UseLoopDelay=false)
   &SpawnBurst_Instantaneous(Use Spawn Probability=false, UseLoopCountLimit=false)
  ]
  particleSpawnScript [&InitializeParticle(
    Position Offset=<15.9 -0.6 -4.5>
    Sprite Size Mode=::NewEnumerator3
    Uniform Sprite Size=&MakeCustomFloatFromBool(Bool=$User.bIsFrontEndPreview, True Float=10, False Float=9)
    UsePositionOffset=true
    Write Material Random=true
   )]
  particleUpdateScript [
   &ParticleState(Kill Particles When Lifetime Has Elapsed=false, Loop Particles Lifetime=true)
   &ScaleColor(
    Curve Index=$System.NormalizedLoopAge
    Linear Color Curve=DataInterface() {propertyValues {
      AlphaCurve {keys [<at 0 1>,<at 1 1>]}
      BlueCurve {keys [<at 0 0>,<at 0.2429 0.6734>,<at 0.5227 1>,<at 0.7851 0>,<at 1 0>]}
      GreenCurve {keys [<at 0 0.0452>,<at 0.2429 0.1165>,<at 0.5227 0.7791>,<at 0.7851 0.5973>,<at 1 0.0452>]}
      RedCurve {keys [<at 0 1>,<at 0.2429 1>,<at 0.5227 0.0248>,<at 0.7851 1>,<at 1 1>]}
     }}
    Scale Mode=::NewEnumerator2
   )
   /SetVariables(Particles.SpriteSize=&Multiply_Vector2D_ByFloat(
     Float=&FloatFromCurve(FloatCurve=DataInterface() {propertyValues {Curve {keys [<at 0 1>,<at 0.5 1.15>]}}}, CurveIndex=$System.Timeline.Time)
     Vector2D=$Particles.Initial.SpriteSize
    )) {isSetParameters=true, module=SetVariables_#1}
   &FN_Light_Attributes(
    Color=$Particles.Color
    Create Color Attribute=true
    Create Exponent Attribute=true
    Create Radius Attribute=true
    Enable Particle Light Enable=true
    Exponent=2
    Light Enabled=true
    Radius=&Lerp_Float(A=80, B=90, Alpha=$System.AmplitudeValue)
    Scale Color=3
   )
   &CameraOffset(Camera Offset Amount=4)
   &DynamicMaterialParameters(
    Index 0 Param 1=&FloatFromCurve(FloatCurve=DataInterface() {propertyValues {Curve {keys [
         {arriveTangent=9.0725, interpMode=RCIM_Cubic, leaveTangent=9.0725, tangentMode=RCTM_User, time=0, value=0.1}
         {arriveTangent=-2.5249, interpMode=RCIM_Cubic, leaveTangent=-0.26, tangentMode=RCTM_Break, time=0.1, value=2.5}
         {arriveTangent=-0.0405, interpMode=RCIM_Cubic, leaveTangent=-0.0405, tangentMode=RCTM_User, time=1, value=0.1}
        ]}}})
    Module.Index0Param0Write=false
    Module.Index0Param1Write=false
    Module.Index0Param2Write=false
    Module.Index0Param3Write=false
   )
  ]
  rendererClasses [&NiagaraSpriteRendererProperties]
  renderers [{rendererClass=&NiagaraSpriteRendererProperties}]
 }
 GlowEdge {
  emitterUpdateScript [
   &EmitterState(Life Cycle Mode=::NewEnumerator1, Loop Behavior=::NewEnumerator1, Loop Duration Mode=::NewEnumerator1, UseLoopDelay=false)
   &SpawnBurst_Instantaneous(Use Spawn Probability=false, UseLoopCountLimit=false)
  ]
  particleSpawnScript [&InitializeParticle(
    Color Mode=::NewEnumerator0
    Mesh Scale Mode=::NewEnumerator3
    Mesh Uniform Scale=0.07
    Position Offset=<15.9 -0.6 -4.5>
    UsePositionOffset=true
    Write Material Random=true
   )]
  particleUpdateScript [
   &ParticleState(Kill Particles When Lifetime Has Elapsed=false, Loop Particles Lifetime=true)
   &CameraOffset(Camera Offset Amount=1)
   &ScaleColor(
    Curve Index=$System.NormalizedLoopAge
    Linear Color Curve=DataInterface() {propertyValues {
      AlphaCurve {keys [<at 0 1>,<at 1 1>]}
      BlueCurve {keys [<at 0 0>,<at 0.2429 0.6734>,<at 0.5227 1>,<at 0.7851 0>,<at 1 0>]}
      GreenCurve {keys [<at 0 0.0452>,<at 0.2429 0.1165>,<at 0.5227 0.7791>,<at 0.7851 0.5973>,<at 1 0.0452>]}
      RedCurve {keys [<at 0 1>,<at 0.2429 1>,<at 0.5227 0.0248>,<at 0.7851 1>,<at 1 1>]}
     }}
    Scale Mode=::NewEnumerator2
   ) {module=ScaleColor001}
   &DynamicMaterialParameters(
    Index 0 Param 1=&FloatFromCurve(
     FloatCurve=DataInterface() {propertyValues {Curve {keys [
         {arriveTangent=-0.0382, interpMode=RCIM_Cubic, leaveTangent=-0.0382, tangentMode=RCTM_User, time=0, value=0.6}
         {arriveTangent=1.1864, interpMode=RCIM_Cubic, leaveTangent=0.4645, tangentMode=RCTM_Break, time=0.1408, value=0.3}
         {arriveTangent=-0.0405, interpMode=RCIM_Cubic, leaveTangent=-0.0405, tangentMode=RCTM_User, time=1, value=0.6}
        ]}}}
     CurveIndex=&MakeCustomFloatFromBool(Bool=$User.bIsFrontEndPreview, True Float=$Particles.NormalizedAge, False Float=$NPC.NPC_Sparks_Audio_Parameters001.BeatLifetime)
    )
    Index 0 Param 2=&FloatFromCurve(FloatCurve=DataInterface() {propertyValues {Curve {keys [
         {arriveTangent=-0.0382, interpMode=RCIM_Cubic, leaveTangent=-0.0382, leaveTangentWeight=0.0333, tangentMode=RCTM_User, time=0, value=0.25}
         {interpMode=RCIM_Cubic, leaveTangent=0.5429, tangentMode=RCTM_Break, time=0.1, value=0.001}
         {arriveTangent=-0.0405, arriveTangentWeight=0.2864, interpMode=RCIM_Cubic, leaveTangent=-0.0405, tangentMode=RCTM_User, time=1, value=0.25}
        ]}}})
    Module.Index0Param0Write=true
    Module.Index0Param1Write=true
    Module.Index0Param2Write=false
    Module.Index0Param3Write=false
   )
  ]
  rendererClasses [&NiagaraMeshRendererProperties]
  renderers [{rendererClass=&NiagaraMeshRendererProperties, rendererIndex=0}]
 }
 Orb {
  emitterUpdateScript [
   &EmitterState(Life Cycle Mode=::NewEnumerator1, Loop Behavior=::NewEnumerator1, Loop Duration Mode=::NewEnumerator1, UseLoopDelay=false)
   &SpawnBurst_Instantaneous(Use Spawn Probability=false, UseLoopCountLimit=false)
  ]
  particleSpawnScript [
   &PartitionParticles(Number Of Partitions=3)
   &InitializeParticle(
    Mesh Scale Mode=::NewEnumerator3
    Mesh Uniform Scale=&MakeCustomFloatFromBool(Bool=$User.bIsFrontEndPreview, True Float=0.07, False Float=0.07)
    Position Offset=<15.9 -0.6 -4.5>
    UsePositionOffset=true
    Write Material Random=true
   )
  ]
  particleUpdateScript [&ParticleState(
    Kill Particles When Lifetime Has Elapsed=false
    Let Infinitely Lived Particles Die When Emitter Deactivates=true
    Loop Particles Lifetime=true
   )]
  rendererClasses [&NiagaraMeshRendererProperties]
  renderers [{rendererClass=&NiagaraMeshRendererProperties, rendererIndex=0}]
 }
 OrbGlow {
  emitterUpdateScript [
   &EmitterState(Life Cycle Mode=::NewEnumerator1, Loop Behavior=::NewEnumerator1, Loop Duration Mode=::NewEnumerator1, UseLoopDelay=false)
   &SpawnBurst_Instantaneous(Use Spawn Probability=false, UseLoopCountLimit=false)
  ]
  particleSpawnScript [&InitializeParticle(
    Color=<rgba 0.7292 0.7292 0.7292 1>
    Position Offset=<15.9 -0.6 -4.5>
    Sprite Size Mode=::NewEnumerator3
    Uniform Sprite Size=20
    UsePositionOffset=true
    Write Material Random=true
   )]
  particleUpdateScript [
   &ParticleState(Kill Particles When Lifetime Has Elapsed=false)
   &ScaleColor(
    Curve Index=$System.NormalizedLoopAge
    Linear Color Curve=DataInterface() {propertyValues {
      AlphaCurve {keys [<at 0 1>,<at 1 1>]}
      BlueCurve {keys [<at 0 0>,<at 0.2429 0.6734>,<at 0.5227 1>,<at 0.7851 0>,<at 1 0>]}
      GreenCurve {keys [<at 0 0.0452>,<at 0.2429 0.1165>,<at 0.5227 0.7791>,<at 0.7851 0.5973>,<at 1 0.0452>]}
      RedCurve {keys [<at 0 1>,<at 0.2429 1>,<at 0.5227 0.0248>,<at 0.7851 1>,<at 1 1>]}
     }}
    Scale Mode=::NewEnumerator2
   )
   /SetVariables(Particles.SpriteSize=&Multiply_Vector2D_ByFloat(
     Float=&FloatFromCurve(FloatCurve=DataInterface() {propertyValues {Curve {keys [<at 0 1>,<at 0.5 1.15>]}}}, CurveIndex=$System.Timeline.Time)
     Vector2D=$Particles.Initial.SpriteSize
    )) {isSetParameters=true, module=SetVariables_#1}
   &FN_Light_Attributes(
    Color=$Particles.Color
    Create Color Attribute=true
    Create Exponent Attribute=true
    Create Radius Attribute=true
    Enable Particle Light Enable=true
    Exponent=2
    Light Enabled=true
    Radius=&Lerp_Float(A=80, B=90, Alpha=$System.AmplitudeValue)
    Scale Color=3
   )
   /SetVariables() {enabled=false, isSetParameters=true}
  ]
  rendererClasses [&NiagaraSpriteRendererProperties, &NiagaraLightRendererProperties]
  renderers [
   {rendererClass=&NiagaraSpriteRendererProperties}
   {rendererClass=&NiagaraLightRendererProperties, rendererIndex=1}
  ]
 }
}
summary {
 systemName=NS_Keytar_RelayStick
 emitters [
  {emitterName=Orb, rendererClasses [&NiagaraMeshRendererProperties]}
  {emitterName=GlowEdge, rendererClasses [&NiagaraMeshRendererProperties]}
  {emitterName=OrbGlow, rendererClasses [&NiagaraSpriteRendererProperties, &NiagaraLightRendererProperties]}
  {emitterName=Energy, rendererClasses [&NiagaraSpriteRendererProperties]}
 ]
 userVariables [User.bIsFrontEndPreview {type=NiagaraBool}]
}
